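# Additive, unlit, two-sided Niagara material for a ring effect in the PulseLancer tactical sprint effects.
# SplineThicken sets line width from a W scalar scaled by a dynamic particle parameter.
# Emissive blends particle color toward a UV-driven hue-shifted rainbow via UseRainbowRing, then multiplies by particle color again.
# An edge mask across U, particle alpha and squared depth fade shape the transparency.
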Material /Game/Effects/Fort_Effects/GameplayPulginsAssets/PulseLancer/TacSprint/M_Ring
graph {
 properties {
  bAutomaticallySetUsageInEditor=false
  bUsedWithNiagaraMeshParticles=true
  bUsedWithNiagaraRibbons=true
  bUsedWithNiagaraSprites=true
  blendMode=BLEND_Additive
  shadingModel=MSM_Unlit
  twoSided=true
  usedShadingModels=Unlit
 }
 nodes [
  n1 = Multiply(A=ScalarParameter() {defaultValue=14.4995, parameterName=W}, B=DynamicParameter() {paramNames [W,Param2,Param3,Param4]}.W)
  MaterialFunctionCall(WidthBase=@n1, WidthTip=@n1) {materialFunction=&SplineThicken}
  n2 = ParticleColor()
  n3 = TextureCoordinate()
  n4 = MaterialFunctionCall(
   Hue Shift Percentage=Multiply(
    A=MaterialFunctionCall(Float2=@n3) {materialFunction=&BreakOutFloat2Components}.G
    B=ScalarParameter() {defaultValue=50, parameterName=HS}
   )
   Texture=VectorParameter() {parameterName=BaseC, defaultValue=<rgba 0.1722 0.4083 1 1>}.RGB
  ) {materialFunction=&HueShift}
  n5 = ScalarParameter() {parameterName=UseRainbowRing}
  MaterialFunctionCall(Emissive=Multiply(A=LinearInterpolate(A=@n2.RGB, B=@n4.Result, Alpha=@n5), B=@n2.RGB)) {materialFunction=&MF_FX_EmissiveAdjustments}
  n6 = MaterialFunctionCall(Float2=TextureCoordinate()) {materialFunction=&BreakOutFloat2Components}
  n7 = VectorParameter() {group=Edge, parameterName=Edge_EdgeMask_InvertControl, channelNames {a="", b=Power, g=Multiply, r=Offset}, defaultValue=<rgba 0 2 3 1>}
  n8 = VectorParameter() {group=Edge, parameterName=Edge_EdgeMask_Control, channelNames {a="", b=Power, g=Multiply, r=Offset}, defaultValue=<rgba 0 2 3 1>}
  n9 = DepthFade(FadeDistance=ScalarParameter() {defaultValue=100, parameterName=DF})
  Multiply(
   A=Multiply(
    A=MaterialFunctionCall!(UorV=@n6.R, Invert_EdgeControl=@n7, NoInvert_EdgeControl=@n8) {materialFunction=/Game/Effects/Niagara/MasterMaterials/Functions/Common/FXCommon_EdgeMask_Ex}.Mask
    B=@n2.A
   )
   B=Multiply(A=@n9, B=@n9)
  )
 ]
}
