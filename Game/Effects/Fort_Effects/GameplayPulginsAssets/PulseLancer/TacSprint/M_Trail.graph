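# Additive, unlit, two-sided Niagara trail material for the PulseLancer TacSprint effect, usable on ribbons, sprites and mesh particles.
# A refined, tinted plasma texture sums with bright grainy "sand" built from contrasted Perlin times tile noise, using panning UV controls.
# Separate refined falloffs along UV1's first axis fade each layer.
# Particle color tints the brightness-offset emissive, while particle alpha multiplies edge masks on both UV1 axes.
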Material /Game/Effects/Fort_Effects/GameplayPulginsAssets/PulseLancer/TacSprint/M_Trail
graph {
 properties {
  bAutomaticallySetUsageInEditor=false
  bUsedWithNiagaraMeshParticles=true
  bUsedWithNiagaraRibbons=true
  bUsedWithNiagaraSprites=true
  blendMode=BLEND_Additive
  shadingModel=MSM_Unlit
  twoSided=true
  usedShadingModels=Unlit
 }
 nodes [
  n1 = TextureObjectParameter() {
   group=T
   parameterName=T_Texture
   samplerType=SAMPLERTYPE_LinearGrayscale
   sortPriority=0
   texture=/Game/Effects/Fort_Effects/Emotes/UltraLight/Texture/T_Plasma_B_Greyscale
  }
  n2 = VectorParameter() {group=T, parameterName=T_UVControl, channelNames {a=VPan, b=UPan, g=VScale, r=UScale}, defaultValue=<rgba 0.6 1 -0.7568 0>}
  n3 = VectorParameter() {parameterName=T_Refine, defaultValue=<rgba 3 10 1 0.5>}
  n4 = VectorParameter() {parameterName=C, defaultValue=<rgba 4.7739 15 9.5044 0.4624>}
  n5 = Multiply(
   A=MaterialFunctionCall(
    Texture=MaterialFunctionCall(Texture=@n1, UVControl_RGB=@n2, UVControl_A=@n2) {materialFunction=&FXCommon_TextureSample}.R
    RGB=@n3.RGB
    A=@n3.RGB
   ) {materialFunction=&MF_FX_Refine}.Result
   B=@n4.RGB
  )
  n6 = TextureCoordinate() {coordinateIndex=1}
  n7 = VectorParameter() {parameterName=TF_Refine, defaultValue=<rgba 15 1 0.05 0.5>}
  n8 = MaterialFunctionCall(
   Texture=OneMinus(None=MaterialFunctionCall(Float2=@n6) {materialFunction=&BreakOutFloat2Components}.R)
   RGB=@n7.RGB
   A=@n7.RGB
  ) {materialFunction=&MF_FX_Refine}
  n9 = TextureObjectParameter() {
   group=Msk
   parameterName=Msk_Texture
   samplerType=SAMPLERTYPE_LinearGrayscale
   sortPriority=0
   texture=&T_SimplePerlin_Greyscale
  }
  n10 = VectorParameter() {group=Msk, parameterName=Msk_UVControl, channelNames {a=VPan, b=UPan, g=VScale, r=UScale}, defaultValue=<rgba 5 0.7381 -0.5 0>}
  n11 = MaterialFunctionCall(Texture=@n9, UVControl_RGB=@n10, UVControl_A=@n10) {materialFunction=&FXCommon_TextureSample}
  n12 = ScalarParameter() {defaultValue=0.1, parameterName=C}
  n13 = TextureObjectParameter() {group=Sand, parameterName=Sand_Texture, sortPriority=0, texture=/Game/Global/Textures/Default/Noise/T_TileNoise_02}
  n14 = VectorParameter() {group=Sand, parameterName=Sand_UVControl, channelNames {a=VPan, b=UPan, g=VScale, r=UScale}, defaultValue=<rgba 18 4 -0.1412 0>}
  n15 = VectorParameter() {parameterName=Sand_Refine, defaultValue=<rgba 3.5 300 0.1 0.5>}
  n16 = MaterialFunctionCall(
   Texture=Multiply(
    A=MaterialFunctionCall(In=@n11.R, Contrast=@n12) {materialFunction=&CheapContrast}.Result
    B=MaterialFunctionCall(Texture=@n13, UVControl_RGB=@n14, UVControl_A=@n14) {materialFunction=&FXCommon_TextureSample}.R
   )
   RGB=@n15.RGB
   A=@n15.RGB
  ) {materialFunction=&MF_FX_Refine}
  n17 = VectorParameter() {parameterName=C2S, defaultValue=<rgba 58.4923 69.0862 111 1>}
  n18 = TextureCoordinate() {coordinateIndex=1}
  n19 = VectorParameter() {parameterName=TF2_Refine, defaultValue=<rgba 4 1 0.05 0.5>}
  n20 = MaterialFunctionCall(
   Texture=OneMinus(None=MaterialFunctionCall(Float2=@n18) {materialFunction=&BreakOutFloat2Components}.R)
   RGB=@n19.RGB
   A=@n19.RGB
  ) {materialFunction=&MF_FX_Refine}
  n21 = Add(
   A=LinearInterpolate(B=Add(A=@n5, B=@n4.A), Alpha=Saturate(None=@n8.Result))
   B=LinearInterpolate(B=Multiply(A=@n16.Result, B=@n17.RGB), Alpha=Saturate(None=@n20.Result))
  )
  n22 = ScalarParameter() {defaultValue=0.3, parameterName=Bs}
  n23 = ParticleColor()
  MaterialFunctionCall(Emissive=Multiply(A=Add(A=@n21, B=@n22), B=@n23.RGB)) {materialFunction=&MF_FX_EmissiveAdjustments}
  n24 = MaterialFunctionCall(Float2=TextureCoordinate() {coordinateIndex=1}) {materialFunction=&BreakOutFloat2Components}
  n25 = VectorParameter() {
   group=EdgeF2
   parameterName=EdgeF2_EdgeMask_Control
   channelNames {a="", b=Power, g=Multiply, r=Offset}
   defaultValue=<rgba 0 4.0741 2.9565 1>
  }
  n26 = VectorParameter() {group=EdgeF2, parameterName=EdgeF2_EdgeMask_InvertControl, channelNames {a="", b=Power, g=Multiply, r=Offset}, defaultValue=<rgba 0 1 2 1>}
  n27 = VectorParameter() {group=TFFade, parameterName=TFFade_EdgeMask_Control, channelNames {a="", b=Power, g=Multiply, r=Offset}, defaultValue=<rgba 0 30 2 1>}
  Multiply(
   A=@n23.A
   B=Multiply(
    A=MaterialFunctionCall!(UorV=@n24.G, Invert_EdgeControl=@n25, NoInvert_EdgeControl=@n26) {materialFunction=/Game/Effects/Niagara/MasterMaterials/Functions/Common/FXCommon_EdgeMask_Ex}.Mask
    B=MaterialFunctionCall(UorV=@n24.R, Invert_EdgeControl=@n27, NoInvert_EdgeControl=@n27) {materialFunction=&FXCommon_EdgeMask_Ex}.Mask
   )
  )
  VectorParameter() {group=TFFade, parameterName=TFFade_EdgeMask_InvertControl, channelNames {a="", b=Power, g=Multiply, r=Offset}, defaultValue=<rgba 0 0 2 1>}
 ]
}
